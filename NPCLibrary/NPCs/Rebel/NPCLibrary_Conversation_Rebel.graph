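# Conversation database for the Rebel NPC, who voices a freedom-first, anti-heist outlook and hands off to shared NPC dialogue.
# The Rebel entry point leads to one Speech node with a default line, a first-ever-conversation line and alternate lines on OutfitReaction priority with empty requirements.
# That node branches through link nodes to the shared NPC Library Quests and BasicServices entry points.
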ConversationDatabase /NPCLibrary/NPCs/Rebel/NPCLibrary_Conversation_Rebel
entry_points [{tag=Conversation.Entry.Rebel, destinations [#1]}]
nodes [
 ConversationEntryPointNode_7 {
  id=#1
  class=ConversationEntryPointNode
  outputs [#2]
  properties {entryTag {tagName=Conversation.Entry.Rebel}}
 }
 Speech_C_4 {
  id=#2
  class=Speech_C
  outputs [#3, #4]
  properties {
   nodeName=""
   generalConfig {
    defaultMessage=The most important thing on this Island isn't Bars or prestige. It's freedom.
    contextualMessages [
     {
      message=I'm not getting tangled up in the crews and their heists.
      contextRequirements [{requirement {class=FortControllerRequirement_IsFirstEverConversationWithNPC}}]
      priority {curve {rowName=Tandem.Speech.Priority.FirstEverInteraction}}
     }
     {
      message=I'm not interested in heists. I don't need the extra heat.
      requirementMatchPolicy=RequireAny
      contextRequirements []
      priority {curve {curveTable=null, rowName=Tandem.Speech.Priority.OutfitReaction}}
     }
     {
      message=The most important thing on this Island isn't Bars or prestige. It's freedom.
      requirementMatchPolicy=RequireAny
      contextRequirements []
      priority {curve {curveTable=null, rowName=Tandem.Speech.Priority.OutfitReaction}}
     }
     {
      message=Some crews get greedy and hoard Bars. They should get taken down a peg.
      requirementMatchPolicy=RequireAny
      contextRequirements []
      priority {curve {curveTable=null, rowName=Tandem.Speech.Priority.OutfitReaction}}
     }
    ]
   }
  }
 }
 ConversationLinkNode_0 {id=#4, class=ConversationLinkNode, properties {remoteEntryTag {tagName=Conversation.Entry.BasicServices}}}
 ConversationLinkNode_1 {id=#3, class=ConversationLinkNode, properties {remoteEntryTag {tagName=Conversation.Entry.NPCLibrary.Quests}}}
]
